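# Conversation database for Season 30 story quest P06 Q01, keyed to the SF entry tag, covering a pipeline explosion and Magneto's arrival.
# Three linear speech nodes run from the entry point and end on a question about the wanderer.
# The first auto-selects as main questline when the NPC has the ShiverFlame tag, the quest is active and obj0 is incomplete.
# Choice text and quest tag reference Brite Raider despite the ShiverFlame requirement.
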ConversationDatabase /BattlePassS30_Quests/Conversations/Story/P06/Quest_01/Conversation_S30_Story_SP_P06_Q01_SF
entry_points [{tag=Conversation.Entry.Season30.Story.P06.Q01.SF, destinations [#1]}]
nodes [
 ConversationEntryPointNode_0 {
  id=#1
  class=ConversationEntryPointNode
  outputs [#2]
  properties {entryTag {tagName=Conversation.Entry.Season30.Story.P06.Q01.SF}}
 }
 Speech_C_10 {
  id=#2
  class=Speech_C
  outputs [#3]
  properties {softTaskIcon=&T_UI_NPC_MainQuest, generalConfig {defaultMessage="The pipeline exploded!"}}
  sub_nodes [
   ConversationChoiceNode_9 {
    class=ConversationChoiceNode
    properties {
     defaultChoiceDisplayText=Talk to Brite Raider
     choiceTags {gameplayTags [
       {tagName=Conversation.Choice.AutoSelectIfValid}
       {tagName=Conversation.Choice.Category.MainQuestline}
       {tagName=Athena.Quests.S30.Conversation.Story.P06.01.BR}
      ]}
    }
   }
   FortConversationRequirement_ControllerRequirement_3 {
    class=FortConversationRequirement_ControllerRequirement
    properties {
     nodeName=Require: NPC Has Conversation Tag
     participantId {tagName=Conversation.Participant.NPC}
     requirement {class=FortControllerRequirement_NPCHasConversationEntryTag, properties {queryTag {tagName=Conversation.Entry.ShiverFlame}}}
    }
   }
   FortConversationRequirement_ControllerRequirement_4 {
    class=FortConversationRequirement_ControllerRequirement
    properties {
     nodeName=Require: Is Quest Active
     requirement {
      class=FortControllerRequirement_IsQuestActive
      properties {questDefinition=/BattlePassS30_Quests/Items/QuestItems/Story/P06/Quest_S30_Story_P06_Q01}
     }
    }
   }
   FortConversationRequirement_ControllerRequirement_5 {
    class=FortConversationRequirement_ControllerRequirement
    properties {
     nodeName=Require: Has Completed Quest Objective
     requirement {
      class=FortControllerRequirement_HasCompletedQuestObjective
      properties {
       bInverseCondition=true
       objectiveBackendName=Quest_S30_Story_P06_Q01_obj0
       questDefinition=&Quest_S30_Story_P06_Q01
      }
     }
    }
   }
  ]
 }
 Speech_C_11 {
  id=#3
  class=Speech_C
  outputs [#4]
  properties {generalConfig {defaultMessage="Magneto showed up!"}}
  sub_nodes [ConversationChoiceNode_12 {class=ConversationChoiceNode, properties {defaultChoiceDisplayText=Ask Brite Raider what happened}}]
 }
 Speech_C_9 {
  id=#4
  class=Speech_C
  outputs []
  properties {generalConfig {defaultMessage=Who is the wanderer?}}
  sub_nodes [ConversationChoiceNode_11 {class=ConversationChoiceNode, properties {defaultChoiceDisplayText=Continue}}]
 }
]
